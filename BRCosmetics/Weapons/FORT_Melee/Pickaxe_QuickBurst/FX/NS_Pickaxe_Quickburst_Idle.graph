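# Idle effect for the QuickBurst pickaxe: a flickering eye glow on the weapon's skeletal mesh.
# A single emitter bursts two particles, each pinned to its own bone or socket by partition index.
# Three sprite renderers and a light renderer stack the glow on each particle, with a waveform-driven flicker scaling brightness.
# User.Variant picks orange default or blue-cyan music colors for the glow layers.
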NiagaraSystem /BRCosmetics/Weapons/FORT_Melee/Pickaxe_QuickBurst/FX/NS_Pickaxe_Quickburst_Idle
dependencies {
 usedDataInterfaces [&NiagaraDataInterfaceSkeletalMesh]
 usedDynamicInputs [&SetBoolByFloatComparison, &ModuloFloat]
 usedModules [
  &FN_VariantManager
  &self:SystemSpawnScript.SystemScriptSource.SystemScriptGraph.NiagaraNodeAssignment_1.SetVariables_#1
  &SystemState
  &Timeline
 ]
}
emitters {EyeGlowStack {
  emitterUpdateScript [
   /Niagara/Noise/Waveform(
    EnableExponent1=false
    GlobalClampMax=false
    GlobalClampMin=false
    UseInput1=false
    UsePhase1=false
    "Waveform [1]":ENiagara_Waveforms=::NewEnumerator12
    "[1] Frequency"=7
   )
   /SetVariables(
    Emitter.Flicker:NiagaraFloat=&Add_Float(A=&Multiply_Float(A=$Emitter.Waveform.Waveform, B=0.5), B=0.7)
    Emitter.FlickerHeavy:NiagaraFloat=&Add_Float(A=&Multiply_Float(A=$Emitter.Waveform.Waveform, B=2), B=0.8)
   ) {isSetParameters=true}
   &EmitterState(
    Inactive Response=::NewEnumerator1
    Life Cycle Mode=::NewEnumerator1
    Loop Behavior=::NewEnumerator1
    Loop Duration Mode=::NewEnumerator1
    UseLoopDelay=false
   )
   &SpawnBurst_Instantaneous(Spawn Count=2, Spawn Time=0.01, Use Spawn Probability=false, UseLoopCountLimit=false)
   /SetVariables(
    Emitter.CoreGlowColor:LinearColor=&MultiplyLinearColor(
     Linear Color=&self:ColorSelector(
      VariantSelector:NiagaraInt32=&MakeIntFromFloat(Float=$User.Variant)
      DefaultColor:LinearColor=<rgba 1 0.5451 0 1>
      MusicColor:LinearColor=<rgba 0 0.6406 1 1>
     )
     Scale Factor=&MakeVector4_FromFloat(Float=&Multiply_Float(A=1, B=$Emitter.Flicker))
    )
    Emitter.CoreGlowSize:Vector2f=&Vector2DFromFloat(Value=1.407)
    Emitter.DarkGlowColor:LinearColor=&MultiplyLinearColor(
     Linear Color=&self:ColorSelector(
      VariantSelector:NiagaraInt32=&MakeIntFromFloat(Float=$User.Variant)
      DefaultColor:LinearColor=<rgba 0.526 0.2173 0.0013 1>
      MusicColor:LinearColor=<rgba 0 0.1436 0.3698 1>
     )
     Scale Factor=&MakeVector4_FromFloat(Float=&Multiply_Float(A=1, B=$Emitter.Flicker))
    )
    Emitter.DarkGlowSize:Vector2f=&Vector2DFromFloat(Value=8.0622)
    Emitter.DepthFade=1
    Emitter.DiffuseGlowColor:LinearColor=&MultiplyLinearColor(
     Linear Color=&self:ColorSelector(
      VariantSelector:NiagaraInt32=&MakeIntFromFloat(Float=$User.Variant)
      DefaultColor:LinearColor=<rgba 2 1.1664 0.0236 1>
      MusicColor:LinearColor=<rgba 0 1.1214 2 1>
     )
     Scale Factor=&MakeVector4_FromFloat(Float=&Multiply_Float(A=1, B=$Emitter.Flicker))
    )
    Emitter.DiffuseGlowSize:Vector2f=&Vector2DFromFloat(Value=3.4401)
    Emitter.LightExponent=5
    Emitter.LightGlowColor:LinearColor=&MultiplyLinearColor(
     Linear Color=&self:ColorSelector(
      VariantSelector:NiagaraInt32=&MakeIntFromFloat(Float=$User.Variant)
      DefaultColor:LinearColor=<rgba 1.5 0.7238 0 1>
      MusicColor:LinearColor=<rgba 0 0.4534 1.5 1>
     )
     Scale Factor=&MakeVector4_FromFloat(Float=&Multiply_Float(A=1, B=$Emitter.Flicker))
    )
    Emitter.LightRadius=50
   ) {isSetParameters=true}
  ]
  particleSpawnScript [
   &PartitionParticles()
   &InitializeParticle(Color=<rgba 1 0.3963 0 0.7>, Sprite Size Mode=::NewEnumerator3, Uniform Sprite Size=120, UsePositionOffset=false, Write Material Random=true)
   &DynamicMaterialParameters(
    Index 0 Param 1=0.2
    Index 0 Param 2=1
    Module.Param0WriteEnabled=true
    Module.Param1WriteEnabled=true
    Module.Param2WriteEnabled=true
    Module.Param3WriteEnabled=true
   )
  ]
  particleUpdateScript [
   &ParticleState(Kill Particles When Lifetime Has Elapsed=false, Let Infinitely Lived Particles Die When Emitter Deactivates=true)
   &SkeletalMeshLocation(
    Bone / Socket Index=$Particles.PartitionParticles.Partition
    MaskSpawnGroup=false
    Mesh Sampling Type=::NewEnumerator5
    Skeletal Mesh=$System.SkeletalMesh
    Socket Sampling Mode=::NewEnumerator5
   )
   &ScaleColor(Scale Alpha=&Multiply_Float(A=$Emitter.Flicker, B=0.3), ScaleA=true, ScaleRGB=true)
   &FN_Light_Attributes(
    Color=$Particles.Color
    Create Color Attribute=true
    Create Exponent Attribute=true
    Create Radius Attribute=true
    Exponent=5
    Radius=4
    Scale Color=&Multiply_Float(A=7, B=$Emitter.Flicker)
   )
  ]
  rendererClasses [&NiagaraSpriteRendererProperties, &NiagaraLightRendererProperties]
  renderers [
   {rendererClass=&NiagaraSpriteRendererProperties}
   {rendererClass=&NiagaraSpriteRendererProperties, rendererIndex=1}
   {rendererClass=&NiagaraSpriteRendererProperties, rendererIndex=2}
   {rendererClass=&NiagaraLightRendererProperties, rendererIndex=3}
  ]
 }}
summary {
 systemName=NS_Pickaxe_Quickburst_Idle
 emitters [{emitterName=EyeGlowStack, rendererClasses [&NiagaraSpriteRendererProperties, &NiagaraLightRendererProperties]}]
 userVariables [
  User.bIsFrontEnd {type=NiagaraBool}
  User.bIsFrontEndPreview {type=NiagaraBool}
  User.ForwardDot {type=NiagaraFloat, defaultValue=0, description=""}
  User.Variant {type=NiagaraFloat, defaultValue=0, description=""}
 ]
}
