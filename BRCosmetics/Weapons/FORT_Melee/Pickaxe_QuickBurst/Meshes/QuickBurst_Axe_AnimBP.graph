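# Anim Blueprint for the QuickBurst pickaxe that spins its wheel bone and adds secondary motion to two tubes.
# Each update, Spin function advances a wrapping angle by a per-frame step, which eases higher during melee, sprinting, sliding or jumping.
# The angle drives an additive bone-space rotation on dyn_wheel, followed by spring-constrained Anim Dynamics on both tubes, simulated relative to the handle.
# Delta time remap interpolates with zero DeltaTime, so it never advances.
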AnimBlueprint /BRCosmetics/Weapons/FORT_Melee/Pickaxe_QuickBurst/Meshes/QuickBurst_Axe_AnimBP
target_skeleton=&/QuickBurst_Axe_Skeleton
graphs [
 AnimGraph {nodes [
   AnimGraphNode_SubInput_0 = Input Pose()
   n1 = "Transform (Modify) Bone - Bone: dyn_wheel"(
    ComponentPose=Local To Component(LocalPose=@AnimGraphNode_SubInput_0)
    Alpha=1
    Translation="0,0,0"
    Rotation=Get Spin result() -> (Spin result:Rotator)
    Scale="1,1,1"
   ) {rotationMode=BMM_Additive, rotationSpace=BCS_BoneSpace, alphaBoolBlend {blendOption=Linear}, boneToModify {boneName=dyn_wheel}}
   n2 = Anim Dynamics - Bone: dyn_tube_r(ComponentPose=@n1, Alpha=1, ExternalForce="0,0,0") {
    angularDampingOverride=0.9999
    angularSpringConstant=3000
    bAngularSpring=true
    bGravityOverrideInSimSpace=true
    bOverrideAngularDamping=true
    bUseGravityOverride=true
    gravityScale=0.3
    simulationSpace=BoneRelative
    alphaBoolBlend {blendOption=Linear}
    boundBone {boneName=dyn_tube_r}
    componentAppliedLinearAccClamp=<3000 3000 3000>
    componentLinearAccScale=<0.1 0.1 0.1>
    componentLinearVelScale=<0.1 0.1 0.1>
    gravityOverride {z=-980}
    physicsBodyDefinitions [{
      boundBone {boneName=dyn_tube_r}
      boxExtents=<10 10 10>
      constraintSetup {angularLimitsMax=<0 10 0>, angularLimitsMin=<0 -40 0>, angularTarget=<0 0 0>}
      localJointOffset=<0 0 -2>
     }]
    relativeSpaceBone {boneName=Melee_TwoHanded_Handle}
   }
   n3 = Anim Dynamics - Bone: dyn_tube_l(ComponentPose=@n2, Alpha=1, ExternalForce="0,0,0") {
    angularDampingOverride=0.9999
    angularSpringConstant=3000
    bAngularSpring=true
    bGravityOverrideInSimSpace=true
    bOverrideAngularDamping=true
    bUseGravityOverride=true
    gravityScale=0.2
    simulationSpace=BoneRelative
    alphaBoolBlend {blendOption=Linear}
    boundBone {boneName=dyn_tube_l}
    componentAppliedLinearAccClamp=<3000 3000 3000>
    componentLinearAccScale=<0.1 0.1 0.1>
    componentLinearVelScale=<0.1 0.1 0.1>
    gravityOverride {z=-980}
    physicsBodyDefinitions [{
      boundBone {boneName=dyn_tube_l}
      boxExtents=<10 10 10>
      constraintSetup {angularLimitsMax=<0 10 0>, angularLimitsMin=<0 -40 0>, angularTarget=<0 0 0>}
      localJointOffset=<0 0 -2>
     }]
    relativeSpaceBone {boneName=Melee_TwoHanded_Handle}
   }
   Output Pose(Result=Component To Local(ComponentPose=@n3)) {layerGroup=None}
  ]}
 EventGraph {nodes [
   n1 = Event BlueprintUpdateAnimation()
   > Set Delta Time X(Delta Time X=@n1.DeltaTimeX)
   > n2 = Sequence()
   > Set Delta time remap(Delta time remap:double=FInterpTo(Current=Get Delta time remap() -> (Delta time remap:double), Target=1, DeltaTime=0, InterpSpeed=1)) -> (Output_Get:double)
   n3 = Spin function!(execute=@n2.then_1, self, Spin:double=Get Spin() -> (Spin:double), Speed:double=Get Speed() -> (Speed:double)) -> (then, Output:double)
   > n4 = Set Spin(Spin:double=@n3) -> (Output_Get:double)
   > Set Spin result(Spin result:Rotator=MakeRotator(Pitch=@n4)) -> (Output_Get:Rotator)
   TryGetPawnOwner()
  ]}
 Spin function {nodes [
   n1 = Spin function!() -> (then, Spin:double=~, Speed:double=~) {is=title K2Node_FunctionEntry}
   > n2 = Set Spin var(Spin var:double=@n1) -> (Output_Get:double)
   n3 = OR Boolean!(A=Get bIsPlayingMeleeAnim(), B=Get bIsSprinting(), C:bool=Get bIsSliding(), D:bool=Get bIsJumping()) -> (ReturnValue)
   > Set Acceleration(Acceleration:double=FInterpTo(Current=Get Acceleration() -> (Acceleration:double), Target=SelectFloat(A=2, B=0.2, bPickA=@n3), DeltaTime=Get Delta Time X(), InterpSpeed=3)) -> (Output_Get:double)
   > n4 = Branch(Condition="float >= float"(A=@n2, B=720))
   > n5 = Set Spin var(Spin var=float - float(A=Get Spin var() -> (Spin var:double), B=720))
   n6 = Get Velocity mult() -> (Velocity mult:double)
   n7 = Get Acceleration()
   n8 = Get Delta time remap()
   n9 = Set Spin var(
    execute=@n4.else
    Spin var=float + float!(A=Get Spin var(), B=float * float(A=@n1.Speed, B=@n8) {is=title K2Node_PromotableOperator}, C:double=float * float(A=@n7, B=@n6)) -> (ReturnValue)
   )
   Return Node!(execute=[@n9.then, @n5.then], Output:double=@n9)
  ]}
]
variables [
 Acceleration {type=number}
 Delta time remap {type=number}
 Delta Time X {type=number}
 Speed {type=number, default=5}
 Spin {type=number}
 Spin result {type=Rotator}
 Spin var {type=number}
 Velocity mult {type=number, default=12}
]
